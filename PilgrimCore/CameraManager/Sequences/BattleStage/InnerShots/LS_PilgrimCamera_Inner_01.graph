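# Level Sequence for an inner battle-stage camera shot of the Pilgrim camera manager, with one camera cut spanning the shot.
# A spawned CineCamera actor carries attach and look-at tracking tracks, alongside a spawned "relocator" actor at the origin, presumably its anchor.
# The camera component stays at a fixed raised offset while keyed roll, pitch and yaw sweep the view and focal length zooms out to wide angle.
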LevelSequence /PilgrimCore/CameraManager/Sequences/BattleStage/InnerShots/LS_PilgrimCamera_Inner_01
bindings [
 CameraComponent {
  id=#1
  parent_id=#2
  possessed_class=CineCameraComponent
  tracks [
   {
    type=MovieSceneFloatTrack
    display_name=Current Focal Length
    sections [{type=MovieSceneFloatSection, channels [None {type=MovieSceneScriptingFloatChannel, default=55, keys [[0,55],[601,12]]}]}]
   }
   {
    type=MovieSceneFloatTrack
    display_name="Manual Focus Distance (Focus Settings)"
    sections [{type=MovieSceneFloatSection, channels [None {type=MovieSceneScriptingFloatChannel, default=100000}]}]
   }
   {
    type=MovieSceneFloatTrack
    display_name=Current Aperture
    sections [{type=MovieSceneFloatSection, channels [None {type=MovieSceneScriptingFloatChannel, default=2.8}]}]
   }
   {
    type=MovieScene3DTransformTrack
    sections [{
      type=MovieScene3DTransformSection
      channels [
       Location.X {type=MovieSceneScriptingDoubleChannel, default=0}
       Location.Y {type=MovieSceneScriptingDoubleChannel, default=0}
       Location.Z {type=MovieSceneScriptingDoubleChannel, default=449}
       Rotation.X {type=MovieSceneScriptingDoubleChannel, default=36, keys [[0,10.0885],[381,36],[602,12.0151]]}
       Rotation.Y {type=MovieSceneScriptingDoubleChannel, default=-33, keys [[0,17.3456],[381,12.1111],[602,-33]]}
       Rotation.Z {type=MovieSceneScriptingDoubleChannel, default=312, keys [[0,0.8411],[602,151.2238]]}
       Scale.X {type=MovieSceneScriptingDoubleChannel, default=1}
       Scale.Y {type=MovieSceneScriptingDoubleChannel, default=1}
       Scale.Z {type=MovieSceneScriptingDoubleChannel, default=1}
      ]
     }]
   }
  ]
 }
 Cine Camera Actor {
  id=#2
  possessed_class=CineCameraActor
  tracks [
   {type=MovieSceneSpawnTrack, sections [{type=MovieSceneSpawnSection, channels [None {type=MovieSceneScriptingBoolChannel, default=true}]}]}
   {
    type=MovieScene3DTransformTrack
    sections [{
      type=MovieScene3DTransformSection
      channels [
       Location.X {type=MovieSceneScriptingDoubleChannel, default=0}
       Location.Y {type=MovieSceneScriptingDoubleChannel, default=0}
       Location.Z {type=MovieSceneScriptingDoubleChannel, default=0}
       Rotation.X {type=MovieSceneScriptingDoubleChannel, default=0}
       Rotation.Y {type=MovieSceneScriptingDoubleChannel, default=0}
       Rotation.Z {type=MovieSceneScriptingDoubleChannel, default=0}
       Scale.X {type=MovieSceneScriptingDoubleChannel, default=1}
       Scale.Y {type=MovieSceneScriptingDoubleChannel, default=1}
       Scale.Z {type=MovieSceneScriptingDoubleChannel, default=1}
      ]
     }]
   }
   {
    type=MovieScene3DAttachTrack
    sections [{type=MovieScene3DAttachSection, end_frame=603, start_frame=602}, {type=MovieScene3DAttachSection, end_frame=602, start_frame=0}]
   }
   {
    type=MovieSceneBoolTrack
    display_name="Enable Look at Tracking (Lookat Tracking Settings)"
    sections [{type=MovieSceneBoolSection, channels [None {type=MovieSceneScriptingBoolChannel, default=true}]}]
   }
   {type=MovieSceneActorReferenceTrack}
  ]
 }
 relocator {
  id=#3
  possessed_class=Actor
  tracks [
   {
    type=MovieSceneSpawnTrack
    sections [
     {type=MovieSceneSpawnSection, channels [None {type=MovieSceneScriptingBoolChannel, default=true}]}
     {type=MovieSceneSpawnSection, channels [None {type=MovieSceneScriptingBoolChannel, default=true}]}
    ]
   }
   {
    type=MovieScene3DTransformTrack
    sections [{
      type=MovieScene3DTransformSection
      channels [
       Location.X {type=MovieSceneScriptingDoubleChannel, default=0}
       Location.Y {type=MovieSceneScriptingDoubleChannel, default=0}
       Location.Z {type=MovieSceneScriptingDoubleChannel, default=0}
       Rotation.X {type=MovieSceneScriptingDoubleChannel, default=0}
       Rotation.Y {type=MovieSceneScriptingDoubleChannel, default=0}
       Rotation.Z {type=MovieSceneScriptingDoubleChannel, default=0}
       Scale.X {type=MovieSceneScriptingDoubleChannel, default=1}
       Scale.Y {type=MovieSceneScriptingDoubleChannel, default=1}
       Scale.Z {type=MovieSceneScriptingDoubleChannel, default=1}
      ]
     }]
   }
  ]
 }
]
display_rate {denominator=1, numerator=30}
tick_resolution {denominator=1, numerator=24000}
tracks [{type=MovieSceneCameraCutTrack, sections [{type=MovieSceneCameraCutSection, end_frame=602, start_frame=0}]}]
