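# Level Sequence for inner shot 02 of the Pilgrim battle-stage camera set, a single-cut camera move of roughly 20 seconds.
# A spawned CineCamera descends and slides along Y while tilting upward and widening from telephoto to very wide, with fixed focus and aperture.
# Two spawned helper actors, one at the origin and one offset, accompany it; camera and offset actor use attach tracks.
# Look-at tracking is enabled through an actor-reference track.
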LevelSequence /PilgrimCore/CameraManager/Sequences/BattleStage/InnerShots/LS_PilgrimCamera_Inner_02
bindings [
 CameraComponent {
  id=#1
  parent_id=#2
  possessed_class=CineCameraComponent
  tracks [
   {
    type=MovieSceneFloatTrack
    display_name=Current Focal Length
    sections [{type=MovieSceneFloatSection, channels [None {type=MovieSceneScriptingFloatChannel, default=55, keys [[0,55],[601,12]]}]}]
   }
   {
    type=MovieSceneFloatTrack
    display_name="Manual Focus Distance (Focus Settings)"
    sections [{type=MovieSceneFloatSection, channels [None {type=MovieSceneScriptingFloatChannel, default=100000}]}]
   }
   {
    type=MovieSceneFloatTrack
    display_name=Current Aperture
    sections [{type=MovieSceneFloatSection, channels [None {type=MovieSceneScriptingFloatChannel, default=2.8}]}]
   }
  ]
 }
 CineCameraActor0 {
  id=#2
  possessed_class=CineCameraActor
  tracks [
   {type=MovieSceneSpawnTrack, sections [{type=MovieSceneSpawnSection, channels [None {type=MovieSceneScriptingBoolChannel, default=true}]}]}
   {
    type=MovieScene3DTransformTrack
    sections [{
      type=MovieScene3DTransformSection
      channels [
       Location.X {type=MovieSceneScriptingDoubleChannel, default=0, const [0,2]}
       Location.Y {type=MovieSceneScriptingDoubleChannel, default=198, keys [[0,0],[601,198]]}
       Location.Z {type=MovieSceneScriptingDoubleChannel, default=179, keys [[0,483],[601,179]]}
       Rotation.X {type=MovieSceneScriptingDoubleChannel, default=0, keys [[0,0],[601,0]]}
       Rotation.Y {type=MovieSceneScriptingDoubleChannel, default=-33.1999, keys [[0,-29.6458],[601,13.7847]]}
       Rotation.Z {type=MovieSceneScriptingDoubleChannel, default=-186.4871, keys [[0,94.316],[601,98.0591]]}
       Scale.X {type=MovieSceneScriptingDoubleChannel, default=1, const [1,2]}
       Scale.Y {type=MovieSceneScriptingDoubleChannel, default=1, const [1,2]}
       Scale.Z {type=MovieSceneScriptingDoubleChannel, default=1, const [1,2]}
      ]
     }]
   }
   {
    type=MovieScene3DAttachTrack
    sections [{type=MovieScene3DAttachSection, end_frame=603, start_frame=602}, {type=MovieScene3DAttachSection, end_frame=602, start_frame=0}]
   }
   {
    type=MovieSceneBoolTrack
    display_name="Enable Look at Tracking (Lookat Tracking Settings)"
    sections [{type=MovieSceneBoolSection, channels [None {type=MovieSceneScriptingBoolChannel, default=true}]}]
   }
   {type=MovieSceneActorReferenceTrack}
  ]
 }
 Actor0 {
  id=#3
  possessed_class=Actor
  tracks [
   {
    type=MovieSceneSpawnTrack
    sections [
     {type=MovieSceneSpawnSection, channels [None {type=MovieSceneScriptingBoolChannel, default=true}]}
     {type=MovieSceneSpawnSection, channels [None {type=MovieSceneScriptingBoolChannel, default=true}]}
    ]
   }
   {
    type=MovieScene3DTransformTrack
    sections [{
      type=MovieScene3DTransformSection
      channels [
       Location.X {type=MovieSceneScriptingDoubleChannel, default=0}
       Location.Y {type=MovieSceneScriptingDoubleChannel, default=0}
       Location.Z {type=MovieSceneScriptingDoubleChannel, default=0}
       Rotation.X {type=MovieSceneScriptingDoubleChannel, default=0}
       Rotation.Y {type=MovieSceneScriptingDoubleChannel, default=0}
       Rotation.Z {type=MovieSceneScriptingDoubleChannel, default=0}
       Scale.X {type=MovieSceneScriptingDoubleChannel, default=1}
       Scale.Y {type=MovieSceneScriptingDoubleChannel, default=1}
       Scale.Z {type=MovieSceneScriptingDoubleChannel, default=1}
      ]
     }]
   }
  ]
 }
 Actor1 {
  id=#4
  possessed_class=Actor
  tracks [
   {
    type=MovieSceneSpawnTrack
    sections [
     {type=MovieSceneSpawnSection, channels [None {type=MovieSceneScriptingBoolChannel, default=true}]}
     {type=MovieSceneSpawnSection, channels [None {type=MovieSceneScriptingBoolChannel, default=true}]}
    ]
   }
   {
    type=MovieScene3DTransformTrack
    sections [{
      type=MovieScene3DTransformSection
      channels [
       Location.X {type=MovieSceneScriptingDoubleChannel, default=-27, const [-27,1]}
       Location.Y {type=MovieSceneScriptingDoubleChannel, default=815, const [815,1]}
       Location.Z {type=MovieSceneScriptingDoubleChannel, default=177, const [177,1]}
       Rotation.X {type=MovieSceneScriptingDoubleChannel, default=0, const [0,1]}
       Rotation.Y {type=MovieSceneScriptingDoubleChannel, default=0, const [0,1]}
       Rotation.Z {type=MovieSceneScriptingDoubleChannel, default=0, const [0,1]}
       Scale.X {type=MovieSceneScriptingDoubleChannel, default=1, const [1,1]}
       Scale.Y {type=MovieSceneScriptingDoubleChannel, default=1, const [1,1]}
       Scale.Z {type=MovieSceneScriptingDoubleChannel, default=1, const [1,1]}
      ]
     }]
   }
   {
    type=MovieScene3DAttachTrack
    sections [{type=MovieScene3DAttachSection, end_frame=603, start_frame=602}, {type=MovieScene3DAttachSection, end_frame=602, start_frame=0}]
   }
  ]
 }
]
display_rate {denominator=1, numerator=30}
tick_resolution {denominator=1, numerator=24000}
tracks [{type=MovieSceneCameraCutTrack, sections [{type=MovieSceneCameraCutSection, end_frame=602, start_frame=0}]}]
